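; Creative powerup device that players collect in a minigame to apply a gameplay effect to themselves, team, class or all players.
; Mesh or radius-sphere overlaps and pickup events feed server-side DoPickup, which checks team/class restrictions and sends an ability activation event per recipient.
; Replicated collectible state drives respawn, visibility, FX and audio; zero respawn time replicates an FX timestamp instead.
; SetupEffectData supplies duration, SetByCaller magnitudes and persist-on-death tags.
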(event Custom|Client_PlayAmbienLoopSound
  ;[1 0.8036 0 0.3] Effects
  (if (Variables|Default|GetbAmbientAudio)
    (if (not (Audio|Components|Audio|IsPlaying (Variables|Components|GetAmbient%20Loop)))
      (Audio|Components|Audio|FadeIn (Variables|Components|GetAmbient%20Loop) 0.5 1.0 0.0 "Linear" :!class /Script/Engine.AudioComponent))
    (else
      (CallFunction|FadeOutAmbientSound))))

(event Custom|Play%20Ambient%20Loop%20Sound
  ;[1 0.8036 0 0.3] Effects
  (if (Variables|Default|GetbAmbientAudio)
    (if (not (Audio|Components|Audio|IsPlaying (Variables|Components|GetAmbient%20Loop)))
      (Audio|Components|Audio|FadeIn (Variables|Components|GetAmbient%20Loop) 0.5 1.0 0.0 "Linear" :!class /Script/Engine.AudioComponent))
    (else
      (CallFunction|FadeOutAmbientSound))))

(event Custom|PlayCollectionFX
  ;[1 0.8036 0 0.3] Effects
  (Utilities|FlowControl|DoOnce false
    (:Completed
      (bind _returnvalue (Effects|SpawnEmitteratLocation :EmitterTemplate (Variables|Default|GetPickup%20VFX) :Location (+ (Transformation|GetWorldLocation (Variables|Components|GetStaticMeshComponent)) (Variables|Default|GetPickupVFXOffset)) :Scale "1.000000,1.000000,1.000000" :bAutoDestroy true :PoolingMethod "AutoRelease" :bAutoActivateSystem true))
      (Effects|Components|ParticleSystem|SetColorParameter _returnvalue "Color" (Variables|Default|GetVFX%20Particle%20Color)))))

(event Custom|ResetPlayCollectionFXGate
  ;[1 0.8036 0 0.3] Effects
  (Utilities|FlowControl|DoOnce false
    (:Completed
      (bind _returnvalue (Effects|SpawnEmitteratLocation :EmitterTemplate (Variables|Default|GetPickup%20VFX) :Location (+ (Transformation|GetWorldLocation (Variables|Components|GetStaticMeshComponent)) (Variables|Default|GetPickupVFXOffset)) :Scale "1.000000,1.000000,1.000000" :bAutoDestroy true :PoolingMethod "AutoRelease" :bAutoActivateSystem true))
      (Effects|Components|ParticleSystem|SetColorParameter _returnvalue "Color" (Variables|Default|GetVFX%20Particle%20Color)))))

(event Custom|Play%20Pick%20Up%20Sound
  ;[1 0.8036 0 0.3] Effects
  (if (Math|Conversions|ToBoolean(Integer) (Variables|Default|GetPickUpAudioKnob))
    (Audio|PlaySoundatLocation (Variables|Default|GetPickupSound) (Transformation|GetActorLocation))))

(event Custom|Fade%20Out%20Ambient%20Sound
  ;[1 0.8036 0 0.3] Effects
  (if (Audio|Components|Audio|IsPlaying (Variables|Components|GetAmbient%20Loop))
    (Audio|Components|Audio|FadeOut (Variables|Components|GetAmbient%20Loop) 0.2 0.0 "Linear" :!class /Script/Engine.AudioComponent)))

(event OnComponentBeginOverlap(StaticMeshComponent)
  (params (OverlappedComponent :type Primitive%20Component%20Object%20Reference) (OtherActor :type Actor%20Object%20Reference) (OtherComp :type Primitive%20Component%20Object%20Reference) (OtherBodyIndex :type Integer) (bFromSweep :type Boolean) (SweepResult :type Hit%20Result%20Structure%20%28by%20ref%29))
  ;[1 1 1 0.5] Overlaps
  (if (< (Variables|Default|GetPickupRadius) 0.0)
    (CallFunction|DoPickup OtherActor)))

(event OnComponentBeginOverlap(Sphere)
  (params (OverlappedComponent :type Primitive%20Component%20Object%20Reference) (OtherActor :type Actor%20Object%20Reference) (OtherComp :type Primitive%20Component%20Object%20Reference) (OtherBodyIndex :type Integer) (bFromSweep :type Boolean) (SweepResult :type Hit%20Result%20Structure%20%28by%20ref%29))
  ;[1 1 1 0.5] Overlaps
  (if (>= (Variables|Default|GetPickupRadius) 0.0)
    (CallFunction|DoPickup OtherActor)))

(event MinigameStateChangedDelegate(FortMinigameLogic)
  (params (Minigame :type Fort%20Minigame%20Object%20Reference) (NewMinigameState :type EFortMinigameState%20Enum :default "PreGame"))
  ;[0 0.2959 1 0.5] State Logic
  (CallFunction|UpdateForMinigameState NewMinigameState)
  (CallFunction|PlayAmbientLoopSound)
  (if (or (and (Utilities|Enum|Equal(Enum) NewMinigameState "PreGame") (Class|CreativeCommonDeviceFunctionLibrary|IsinPlayMode self)) (Utilities|Enum|Equal(Enum) NewMinigameState "WaitingForCameras"))
    (CallFunction|SetIsPowerupCollectible (Variables|Default|GetKnob_SpawnOnMinigameStart) false) ; Set initial hidden state when play mode changes.
  ))

(event MinigameStartedDelegate(FortMinigameLogic)
  ;[0 0.2959 1 0.5] State Logic
  (CallFunction|SetIsPowerupCollectible (or (Variables|Default|GetKnob_SpawnOnMinigameStart) (== (Variables|Default|GetRespawnTime) 0.0)) false)
  (CallFunction|PlayAmbientLoopSound)
  (CallFunction|HandleRespawnTimer (not (Variables|Default|GetKnob_SpawnOnMinigameStart))) ; If we've not spawned it initially, start the respawn timer here
  (CallFunction|ApplySubscriptiontoPlayerChangesTeamorClass true))

(event MinigameEndedDelegate(FortMinigameLogic)
  ;[0 0.2959 1 0.5] State Logic
  (CallFunction|SetIsPowerupCollectible true false)
  (CallFunction|PlayAmbientLoopSound)
  (CallFunction|ApplySubscriptiontoPlayerChangesTeamorClass false))

(event EventBeginPlay
  (bind _fortminigamelogic (Variables|Components|GetFortMinigameLogic))
  (bind _currentminigame (Variables|/Script/FortniteGame.FortMinigameLogicComponent|GetCurrentMinigame _fortminigamelogic))
  ;[0.1018 1 0 0.5] Initialization
  (Networking|CanExecuteCosmeticEvents
    (:True
      (CallFunction|ClientPlayAmbienLoopSound))
    (:False
      _))
  (CallFunction|InitializeRestriction)
  (Utilities|IsValid _currentminigame
    (:Is%20Valid
      (CallFunction|UpdateForMinigameState (Variables|/Script/FortniteGame.FortMinigame|GetCurrentState _currentminigame))
      (if (Class|CreativeCommonDeviceFunctionLibrary|IsinPlayMode _currentminigame)
        (CallFunction|ApplySubscriptiontoPlayerChangesTeamorClass true)))))

(event EventEndPlay
  (params (EndPlayReason :type EEndPlayReason%20Enum :default "Destroyed"))
  ;[0 1 0.6062 0.5] Player Exit Minigame
  (CallFunction|ApplySubscriptiontoPlayerChangesTeamorClass false))

(event OnPropertiesLoaded(ToyOptionsComponent)
  ;[0.1018 1 0 0.5] Initialization
  (CallFunction|UpdateforPropertyChange)
  (Networking|CanExecuteCosmeticEvents
    (:True
      (CallFunction|ClientPlayAmbienLoopSound))
    (:False
      _))
  (if (Variables|Default|GetbPlayFXOnPlayer)
    (if (not (GameplayTags|HasMatchingGameplayTag self "(TagName=\"Creative.UserOptions.PowerupDevice.ShowFXOnPlayer\")"))
      (Ability|GameplayEffect|AddLooseGameplayTags self (Variables|Default|GetShowFXGamePlayTags)))
    (else
      (Ability|GameplayEffect|RemoveLooseGameplayTags self (Variables|Default|GetShowFXGamePlayTags)))))

(event OnAnyPropertyChanged(ToyOptionsComponent)
  ;[0.1018 1 0 0.5] Initialization
  (CallFunction|UpdateforPropertyChange)
  (Networking|CanExecuteCosmeticEvents
    (:True
      (CallFunction|ClientPlayAmbienLoopSound))
    (:False
      _))
  (if (Variables|Default|GetbPlayFXOnPlayer)
    (if (not (GameplayTags|HasMatchingGameplayTag self "(TagName=\"Creative.UserOptions.PowerupDevice.ShowFXOnPlayer\")"))
      (Ability|GameplayEffect|AddLooseGameplayTags self (Variables|Default|GetShowFXGamePlayTags)))
    (else
      (Ability|GameplayEffect|RemoveLooseGameplayTags self (Variables|Default|GetShowFXGamePlayTags)))))

(event Custom|SpawnEvent
  (params (Instigator :type Controller%20Object%20Reference))
  ;[0.4921 0 1 0.5] Gameplay Functions
  (CallFunction|SetIsPowerupCollectible true false))

(event Custom|DespawnEvent
  (params (Instigator :type Controller%20Object%20Reference))
  ;[0.4921 0 1 0.5] Gameplay Functions
  (CallFunction|SetIsPowerupCollectible false false))

(event Custom|PickupEvent
  (params (Instigator :type Controller%20Object%20Reference))
  (bind _returnvalue (Pawn|GetControlledPawn Instigator))
  ;[0.4921 0 1 0.5] Gameplay Functions
  (Utilities|IsValid Instigator
    (:Is%20Valid
      (Utilities|IsValid _returnvalue
        (:Is%20Valid
          (CallFunction|DoPickup _returnvalue))))))

(event OnAddedtoMinigame(FortMinigameLogic)
  (params (Minigame :type Fort%20Minigame%20Object%20Reference))
  ;[1 0.3619 0] Update Collision Overlaps when minigame is replicated
  (CallFunction|UpdateSphereCollision (Variables|/Script/FortniteGame.FortMinigame|GetCurrentState Minigame)))

(event Custom|On%20Powerup%20Removed
  (params (For%20Controller :type Controller%20Object%20Reference)))

(fn AllowInverseScaling
  (return false))

(fn ApplyPowerupEffecttoPlayer
  (params (Player%20Pawn :type Fort%20Player%20Pawn%20Object%20Reference))
  (Ability|SendGameplayEventtoActor self (Variables|Default|GetPowerupAbilityActivationTag) (Utilities|Struct|MakeGameplayEventData (Variables|Default|GetPowerupAbilityActivationTag) Player%20Pawn 0 self 0 "()" "(GameplayTags=)" "(GameplayTags=)" 2e+01))
  (if (Variables|Default|GetPersistOnDeath)
    (if (not (GameplayTags|HasMatchingGameplayTag (Variables|Default|GetPlayer%20Pawn) "(TagName=\"Granted.Creative.Powerup.PersistOnDeathCleanup\")"))
      (Ability|Advanced|GrantandActivateAbilityonSelfwithParams (Ability|GetAbilitySystemComponent (Variables|Default|GetPlayer%20Pawn)) "/Game/Creative/Devices/Common/CreativePowerups/Abilites/GA_PersistOnDeath_Cleanup.GA_PersistOnDeath_Cleanup_C" 0 (Utilities|Struct|MakeGameplayEventData "(TagName=\"\")" self)))))

(fn ApplySubscriptionToPlayerChangesTeamOrClass
  (params (Subscribe :type Boolean))
  (bind _outputdelegate (EventDispatchers|CreateEvent :!event Update%20Visibility))
  (bind _fortminigamelogic (Variables|Components|GetFortMinigameLogic))
  (bind _currentminigame (Variables|/Script/FortniteGame.FortMinigameLogicComponent|GetCurrentMinigame _fortminigamelogic))
  (Networking|CanExecuteCosmeticEvents
    (:True
      (for _array_element (Fortnite|GetLocalFortPlayerControllers true)
        (bind _asfort_player_state_athena (Utilities|Casting|CastToFortPlayerStateAthena (Variables|/Script/Engine.Controller|GetPlayerState _array_element)))
        (if (Variables|GetSubscribe)
          (Player|BindEventtoOnTeamIndexChangedDelegate _asfort_player_state_athena _outputdelegate)
          (Minigame|BindEventtoOnAnyLocalPlayerChangedClass _currentminigame _outputdelegate)
          (else
            (Player|UnbindEventfromOnTeamIndexChangedDelegate _asfort_player_state_athena _outputdelegate)
            (Minigame|UnbindEventfromOnAnyLocalPlayerChangedClass _currentminigame _outputdelegate)))))))

(fn DoPickup
  (params (Actor :type Actor%20Object%20Reference))
  (Networking|SwitchHasAuthority
    (:Authority
      (Utilities|IsCurrentMinigameActive
        (:Active
          (if (CallFunction|IsCollectable)
            (bind _playerpawn (CallFunction|GetPlayerPawnfromActor Actor))
            (bind _output_get (Variables|Default|SetPlayer%20Pawn _playerpawn))
            (Utilities|IsValid _output_get
              (:Is%20Valid
                (bind _valid_controller (DEPRECATED|ActorHasValidController (Variables|Components|GetCreativeUseRestriction) (Variables|GetActor) true true))
                (if _valid_controller
                  (bind _players (Default|GetApplytoList))
                  (for _array_element _players
                    (CallFunction|ApplyPowerupEffectToPlayer _array_element))
                  (CallFunction|SetIsPowerupCollectible false)
                  (bind _player_pawn (Variables|Default|GetPlayer%20Pawn)
                    (:then
                      (LocalVariables|SetPawn%20Controller (Pawn|GetController _player_pawn)))
                    (:else
                      _))
                  (GameplayEvent|BroadcastEvent (Variables|UserOptions|Events|GetOn%20Item%20Picked%20Up) 0 (Utilities|Struct|MakeGameplayEventPickedUp (LocalVariables|GetPawn%20Controller))))))))))))

(fn ForceGridAlignment
  (return false))

(fn GetAdditionalSelectableActors
  (return 0))

(fn GetApplyToList
  (bind _player_pawn (Variables|Default|GetPlayer%20Pawn))
  (bind _playerstate (Variables|/Script/Engine.Pawn|GetPlayerState _player_pawn))
  (bind _asfort_player_state (Utilities|Casting|CastToFortPlayerState _playerstate))
  (Utilities|FlowControl|Sequence
    (:then_0
      (bind _fortminigamelogic (Variables|Components|GetFortMinigameLogic)
        (:then
          (bind _currentminigame (Variables|/Script/FortniteGame.FortMinigameLogicComponent|GetCurrentMinigame _fortminigamelogic)
            (:then
              (Utilities|IsValid (Minigame|GetFNEClassComponent _currentminigame)
                (:Is%20Valid
                  (switch Utilities|FlowControl|Switch|SwitchonEnum_Creative_Powerup_Applicability (Variables|Default|GetApplyToKnob)
                    (:NewEnumerator0
                      (LocalVariables|SetApplyToList (Utilities|Array|MakeArray (Variables|Default|GetPlayer%20Pawn))))
                    (:NewEnumerator1
                      (bind _players (CallFunction|GetPlayersbyTeam (Team|GetActorTeam (Variables|Default|GetPlayer%20Pawn))))
                      (LocalVariables|SetApplyToList _players))
                    (:NewEnumerator2
                      (bind _players_2 (Default|GetPlayersbyClassSlot (Minigame|GetClassSlotforPlayer (Minigame|GetFNEClassComponent _currentminigame) _asfort_player_state false)))
                      (LocalVariables|SetApplyToList _players_2))
                    (:NewEnumerator3
                      (bind _players_3 (Default|GetPlayersbyClassSlot (Minigame|GetClassSlotforPlayer (Minigame|GetFNEClassComponent _currentminigame) _asfort_player_state false)))
                      (for _array_element _players_3
                        (if (Math|Byte|Equal(Byte) (Team|GetActorTeam _array_element) (Team|GetActorTeam (Variables|Default|GetPlayer%20Pawn)))
                          (Utilities|Array|Add (LocalVariables|GetApplyToList) _array_element))))
                    (:NewEnumerator4
                      (for _array_element_2 (Minigame|GetParticipatingPlayers (Variables|/Script/FortniteGame.FortMinigameLogicComponent|GetCurrentMinigame (Variables|Components|GetFortMinigameLogic)))
                        (Utilities|Array|Add (LocalVariables|GetApplyToList) (Utilities|Casting|CastToPlayerPawn_Athena (PlayerState|GetCurrentPawn _array_element_2)))))))))
            (:else)))
        (:else)))
    (:then_1
      (for _array_element_3 (LocalVariables|GetApplyToList)
        (if (not (FortPawn|IsDead _array_element_3))
          (Utilities|Array|Add (LocalVariables|GetApplyToListWithoutDead) _array_element_3))))
    (:then_2
      (return (LocalVariables|GetApplyToListWithoutDead)))))

(fn GetBlockPitchAndRoll
  (return false))

(fn GetMagnitudeTags
  (return (Utilities|Map|MakeMap (Variables|Default|GetMagnitudeTag) (Variables|Default|GetEffect%20Magnitude))))

(fn GetPlayerPawnFromActor
  (params (Actor :type Actor%20Object%20Reference))
  (bind _asplayer_pawn_athena (Utilities|Casting|CastToPlayerPawn_Athena Actor)
    (:then
      (return _asplayer_pawn_athena))
    (:CastFailed
      (bind _asfort_athena_vehicle (Utilities|Casting|CastToFortAthenaVehicle Actor)
        (:then
          (Utilities|IsValid (Vehicle|GetDriver _asfort_athena_vehicle)
            (:Is%20Valid
              (bind _asplayer_pawn_athena_2 (Utilities|Casting|CastToPlayerPawn_Athena (Vehicle|GetDriver _asfort_athena_vehicle))
                (:then
                  (return _asplayer_pawn_athena_2))
                (:CastFailed
                  _)))
            (:Is%20Not%20Valid
              _)))
        (:CastFailed
          _))
      (return 0))))

(fn GetPlayersByClassSlot
  (params (ClassSlot :type Byte))
  (Utilities|Array|Clear (LocalVariables|GetSameClassSlot%20Players))
  (bind _fortminigamelogic (Variables|Components|GetFortMinigameLogic)
    (:then
      (bind _currentminigame (Variables|/Script/FortniteGame.FortMinigameLogicComponent|GetCurrentMinigame _fortminigamelogic)
        (:then
          (Utilities|IsValid (Minigame|GetFNEClassComponent _currentminigame)
            (:Is%20Valid
              (LocalVariables|SetLocalMinigameFNEClassComponent (Minigame|GetFNEClassComponent _currentminigame))
              (for _array_element (Minigame|GetParticipatingPlayers _currentminigame)
                (if (Math|Byte|Equal(Byte) (Minigame|GetClassSlotforPlayer (LocalVariables|GetLocalMinigameFNEClassComponent) _array_element false) (Variables|GetClassSlot))
                  (bind _asplayer_pawn_athena (Utilities|Casting|CastToPlayerPawn_Athena (Variables|/Script/Engine.PlayerState|GetPawnPrivate _array_element)))
                  (Utilities|Array|Add (LocalVariables|GetSameClassSlot%20Players) _asplayer_pawn_athena)))
              (return (LocalVariables|GetSameClassSlot%20Players)))
            (:Is%20Not%20Valid
              _)))
        (:else
          _)))
    (:else
      _))
  (return (LocalVariables|GetSameClassSlot%20Players)))

(fn GetPlayersByTeam
  (params (RequiredTeamIndex :type Byte))
  (for _array_element (Minigame|GetParticipatingPlayers (Variables|/Script/FortniteGame.FortMinigameLogicComponent|GetCurrentMinigame (Variables|Components|GetFortMinigameLogic)))
    (if (Math|Byte|Equal(Byte) (Team|GetActorTeam _array_element) (Variables|GetRequiredTeamIndex))
      (Utilities|Array|Add (LocalVariables|GetSameTeamPlayers) (Utilities|Casting|CastToPlayerPawn_Athena (PlayerState|GetCurrentPawn _array_element)))))
  (return (LocalVariables|GetSameTeamPlayers)))

(fn GetScaleMax
  (params (DefaultScaleMax :type Float%20%28single-precision%29))
  (return DefaultScaleMax))

(fn GetScaleMin
  (params (DefaultScaleMin :type Float%20%28single-precision%29))
  (return DefaultScaleMin))

(fn GetSpawnedActors
  (return 0))

(fn HandleRespawnTimer
  (params (Start :type Boolean))
  (Networking|SwitchHasAuthority
    (:Authority
      (if Start
        (bind _returnvalue (Utilities|Time|SetTimerbyEvent (EventDispatchers|CreateEvent :!event ResetPowerupCollectible_TimerDelegate) (Variables|Default|GetRespawnTime)))
        (Variables|Default|SetRespawnTimerHandle _returnvalue)
        (else
          (Utilities|Time|ClearandInvalidateTimerbyHandle (Variables|Default|GetRespawnTimerHandle)))))))

(fn HandleFX
  (Rendering|Material|SetMaterial (Variables|Components|GetFX) 0 0))

(fn HasAdditionalSelectableActors
  (return false))

(fn HasAdditionalValidPlacementLogic
  ;[1 0.8399 0.1063] By default, powerups don't need additional placement logic (i.e. return false so IsValidPlacement will not be called).
  (return false))

(fn InitializeRestriction
  (Initialization|InitializeComponent (Variables|Components|GetCreativeUseRestriction) (Variables|Components|GetFortMinigameLogic) 0 0 (Variables|Default|GetSelectedTeam) (Variables|Default|GetSelectedClass) 0 1 :!class /Game/Creative/Devices/Common/Components/Creative_UseRestriction_Component.Creative_UseRestriction_Component_C))

(fn IsCollectable
  (return (Utilities|Array|Get(acopy) (Variables|Default|GetIsPowerupCollectible)) (Utilities|Array|Get(acopy) (Variables|Default|GetIsPowerupCollectible) 1)))

(fn IsTargetableByLinearMover
  (return true))

(fn IsValidPlacement
  ;[1 0.9588 0.2402] Will not be called by default because HasAdditionalValidPlacementLogic returns false by default.
  (return false))

(fn MakeTagIndexArray
  (params (%5B0%5D :type Integer) (%5B1%5D :type Integer) (%5B2%5D :type Integer))
  ;[0 0.7015 1 0.25] Returns the array of tags, but if the previous tag is not set, we want to ignore the current tag. This is because we hide those from the player.
  (return (Utilities|Array|MakeArray %5B0%5D (Math|Integer|SelectInt %5B1%5D 0 (!= %5B0%5D 0)) (Math|Integer|SelectInt %5B2%5D 0 (!= (Math|Integer|SelectInt %5B1%5D 0 (!= %5B0%5D 0)) 0)))))

(fn OnPreviewEnd
  ;[1 0.7418 0.108] Keep track of if this actor is the current move tool preview actor.
  (Variables|Default|SetbIsCreativeMoveToolPreview false)
  (return false))

(fn OnPreviewStart
  ;[0.4802 1 0.3444] Keep track of if this actor is the current move tool preview actor.
  (Variables|Default|SetbIsCreativeMoveToolPreview true)
  (return false))

(fn OnRep_IsPowerupCollectible
  (bind (_is_collectable _has_been_picked_up) (CallFunction|IsCollectable))
  (Networking|CanExecuteCosmeticEvents
    (:True
      (Default|UpdateVisibility)
      (if (or _is_collectable (not _has_been_picked_up))
        (CallFunction|ResetPlayCollectionFXGate)
        (else
          (CallFunction|PlayCollectionFX)
          (CallFunction|PlayPickUpSound)
          (CallFunction|FadeOutAmbientSound))))
    (:False
      (Networking|SwitchHasAuthority
        (:Authority
          (Networking|ForceNetUpdate))))))

(fn OnRep_PlayFXTime
  (Networking|SwitchHasAuthority
    (:Authority
      (Networking|ForceNetUpdate))
    (:Remote
      _))
  (if (> (+ (Variables|Default|GetPlayFXTime) (Variables|Default|GetFX%20Playtime)) (Fortnite|Sync|Time|GetServerWorldTimeSeconds))
    (CallFunction|PlayCollectionFX)
    (CallFunction|PlayPickUpSound)
    (CallFunction|ResetPlayCollectionFXGate)))

(fn ResetPowerupCollectible_TimerDelegate
  (CallFunction|SetIsPowerupCollectible true)
  (CallFunction|PlayAmbientLoopSound))

(fn SetIsPowerupCollectible
  (params (IsCollectible :type Boolean) (Has%20Been%20Picked%20Up :type Boolean :default true))
  (bind (_is_collectable _has_been_picked_up) (CallFunction|IsCollectable))
  (Networking|SwitchHasAuthority
    (:Authority
      (if (Math|Boolean|NotEqual(Boolean) (CallFunction|IsCollectable) IsCollectible)
        (if (or (!= (Variables|Default|GetRespawnTime) 0.0) (not (Variables|GetHas%20Been%20Picked%20Up)))
          (Variables|Default|SetIsPowerupCollectible (Utilities|Array|MakeArray (Variables|GetIsCollectible) (Variables|GetHas%20Been%20Picked%20Up))) ; 0 = Is Collectable\n1 = Play FX
          (CallFunction|HandleRespawnTimer (not (or _is_collectable (not _has_been_picked_up))))
          (else
            (Variables|Default|SetPlayFXTime (Fortnite|Sync|Time|GetServerWorldTimeSeconds))))))))

(fn SetupEffectData
  (params (EffectHandle :type Gameplay%20Effect%20Spec%20Handle%20Structure) (EffectClass :type Gameplay%20Effect%20Class%20Reference))
  (bind _returnvalue (Ability|GameplayEffect|SetDuration EffectHandle (Variables|Default|GetEffectDuration)))
  (bind _output_get (LocalVariables|SetLocalMagnitudeTags (CallFunction|GetMagnitudeTags)))
  (bind _keys (Utilities|Map|Keys _output_get))
  (for _array_element _keys
    (Ability|GameplayEffect|AssignTagSetbyCallerMagnitude _returnvalue _array_element (Utilities|Map|Find _output_get _array_element)))
  (bind _returnvalue_2 (Ability|GameplayEffect|AddGrantedTags _returnvalue (Variables|Default|GetAdditional%20Tags)))
  (bind _returnvalue_3 (Ability|GameplayEffect|AddAssetTags _returnvalue_2 (Variables|Default|GetAdditional%20Asset%20Tags)))
  (if (Variables|Default|GetPersistOnDeath)
    (bind _returnvalue_4 (Ability|GameplayEffect|AddAssetTag _returnvalue_3 "(TagName=\"Gameplay.Effect.PersistOnDeath\")"))
    (return _returnvalue_4 true)
    (else
      (return _returnvalue_3 true))))

(fn ShouldBlockPlacement
  (params (MoveTool :type Fort%20Creative%20Move%20Tool%20Object%20Reference))
  (return false))

(fn ShouldInvisiblePropsBeAllowed
  (return false))

(fn UpdateforMinigameState
  (params (MinigameState :type EFortMinigameState%20Enum :default "PreGame"))
  (CallFunction|UpdateSphereCollision MinigameState)
  (Default|UpdateVisibility))

(fn UpdateVisibility
  (Networking|CanExecuteCosmeticEvents
    (:True
      (if (Class|CreativeCommonDeviceFunctionLibrary|IsinPlayMode (Variables|/Script/FortniteGame.FortMinigameLogicComponent|GetCurrentMinigame (Variables|Components|GetFortMinigameLogic)))
        (if (CallFunction|IsCollectable)
          (switch Utilities|FlowControl|Switch|SwitchonEnum_Creative_Powerup_Visibility (Variables|Default|GetVisibleToKnob)
            (:NewEnumerator0
              (Rendering|SetActorHiddenInGame true))
            (:NewEnumerator1
              (Rendering|SetActorHiddenInGame false))
            (:NewEnumerator2
              (for _array_element (Fortnite|GetLocalFortPlayerControllers true)
                (Rendering|SetActorHiddenInGame (not (Class|CreativeUseRestrictionComponent|StateChecks (Variables|Components|GetCreativeUseRestriction) _array_element true))))))
          (else
            (Rendering|SetActorHiddenInGame true)))
        (else
          (Rendering|SetActorHiddenInGame false))))))

(fn UpdateForPropertyChange
  (Effects|Components|ParticleSystem|SetColorParameter (Variables|Components|GetFX) "Color" (Variables|Default|GetVFX%20Particle%20Color))
  (UEFNConversionHelper|SetComponentHiddenEditor (Variables|Components|GetDisableMesh) (not (Variables|Default|GetDisablesEffect)))
  (CallFunction|InitializeRestriction)
  ;[1 1 1] Update Collision Overlaps on Property Load
  (bind _fortminigamelogic (Variables|Components|GetFortMinigameLogic)
    (:then
      (bind _currentminigame (Variables|/Script/FortniteGame.FortMinigameLogicComponent|GetCurrentMinigame _fortminigamelogic)
        (:then
          (CallFunction|UpdateSphereCollision (Variables|/Script/FortniteGame.FortMinigame|GetCurrentState _currentminigame)))
        (:else)))
    (:else)))

(fn UpdateSphereCollision
  (params (MinigameState :type EFortMinigameState%20Enum :default "PreGame"))
  (bind _staticmeshcomponent (Variables|Components|GetStaticMeshComponent))
  (bind _sphere (Variables|Components|GetSphere))
  (if (and (> (Variables|Default|GetPickupRadius) 0.0) (Utilities|Enum|Equal(Enum) MinigameState "InProgress"))
    (Components|Sphere|SetSphereRadius (Variables|Components|GetSphere) (* (Variables|Default|GetPickupRadius) 1e+02) true)
    (Collision|SetCollisionEnabled _staticmeshcomponent "NoCollision")
    (Collision|SetCollisionEnabled _sphere "QueryOnly")
    (else
      (Collision|SetCollisionEnabled _staticmeshcomponent "QueryOnly")
      (Collision|SetCollisionEnabled _sphere "NoCollision"))))

(fn ConstructionScript)
