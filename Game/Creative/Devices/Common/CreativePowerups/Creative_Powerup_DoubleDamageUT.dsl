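; Blueprint for the Double Damage UT powerup in Creative; gameplay behavior is not defined here and presumably comes from its parent class.
; On BeginPlay, where cosmetic events can run, it randomizes a custom primitive data float and the mesh's starting yaw, varying each instance's look.
; Overrides supply a purple effect color with preset enum options, empty custom effect data, and allow linear movers to target it.
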(event EventBeginPlay
  (|CallParentFunction)
  (Networking|CanExecuteCosmeticEvents
    (:True
      (Rendering|Material|SetCustomPrimitiveDataFloat (Variables|Components|GetStaticMeshComponent) 31 (Math|Random|RandomFloatinRange 0.0 1e+02))
      (Transformation|SetRelativeRotation (Variables|Components|GetStaticMeshComponent) (Math|Rotator|MakeRotator 0.0 0.0 (Math|Random|RandomFloatinRange 0.0 3.6e+02)) false false))))

(event Collision|EventActorBeginOverlap
  (params (OtherActor :type Actor%20Object%20Reference)))

(event EventTick
  (params (DeltaSeconds :type Float%20%28single-precision%29)))

(fn GetCreativeCustomEffectData
  (return 0 0 "None" 0 0))

(fn GetCreativeEffectData
  (return "(R=0.180392,G=0.031373,B=0.282353,A=1.000000)" "NewEnumerator0" "NewEnumerator0" "NewEnumerator1"))

(fn IsTargetableByLinearMover
  (return true))

(fn ConstructionScript
  (|CallParentFunction))
