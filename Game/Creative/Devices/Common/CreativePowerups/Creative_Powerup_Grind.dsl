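; Creative device powerup for Grind, built on a parent class that supplies the base gameplay effect setup.
; SetupEffectData calls the parent, then adds a suppress-visuals asset tag to the effect spec when ShouldShowGrindEffect is off.
; IsTargetableByLinearMover returns true, while the BeginPlay, overlap and Tick events are empty.
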(event EventBeginPlay)

(event Collision|EventActorBeginOverlap
  (params (OtherActor :type Actor%20Object%20Reference)))

(event EventTick
  (params (DeltaSeconds :type Float%20%28single-precision%29)))

(fn IsTargetableByLinearMover
  (return true))

(fn SetupEffectData
  (params (EffectHandle :type Gameplay%20Effect%20Spec%20Handle%20Structure) (EffectClass :type Gameplay%20Effect%20Class%20Reference))
  (bind _localeffecthandleref (LocalVariables|GetLocalEffectHandleRef))
  (bind _outeffecthandle (|CallParentFunction EffectHandle))
  (LocalVariables|SetLocalEffectHandleRef _outeffecthandle)
  (if (not (Variables|Default|GetShouldShowGrindEffect))
    (Ability|GameplayEffect|AddAssetTag _localeffecthandleref "(TagName=\"Creative.Toys.Abilites.Grind.SupressVisuals\")"))
  (return _localeffecthandleref true))

(fn ConstructionScript
  (|CallParentFunction))
