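; Creative powerup device that supplies a selectable visual gameplay effect with color and visibility settings, tinting its own mesh and particles to match.
; The color knob selects a fixed palette index, team-based color or friendly color from CreativeTeamColors.
; Non-fixed colors bind cosmetic updates to team-color and local-player team changes, unbound on destruction.
; UpdatePowerupColors reattaches the mesh MID as a Move Tool fix and restarts FX to apply new colors.
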(event EventUpdateForPropertyChange
  ;[0 1 0.1197 0.5] Binding and unbinding events to update color when team colors or player team changes.
  (|CallParentFunction)
  (CallFunction|UpdateColor)
  (Networking|CanExecuteCosmeticEvents
    (:True
      (CallFunction|BindOrUnbindTeamColorUpdates (Utilities|Enum|NotEqual(Enum) (CallFunction|ColorType) "NewEnumerator1")))))

(event EventDestroyed
  ;[1 0.06044 0] Unbind events when this actor is removed.
  (CallFunction|BindOrUnbindTeamColorUpdates false))

(fn BindorUnbindTeamColorUpdates
  (params (Bind :type Boolean))
  (bind _outputdelegate (EventDispatchers|CreateEvent :!event Update%20Color))
  (bind _fortminigamelogic (Variables|Components|GetFortMinigameLogic))
  ;[1 1 1] We need to listen for when either the team colors change or when the player changes teams. These combined cover the options of when we want to update the color of the device.
  ;
  (if Bind
    (Teams|RefreshTeamChangeBinding _fortminigamelogic false _outputdelegate)
    (bind _asathena_player_controller (Utilities|Casting|CastToAthena_PlayerController (Fortnite|GetFirstLocalFortPlayerController true)))
    (FortPlayerControllerAthena|BindEventtoOnLocalPlayerChangedTeams _asathena_player_controller _outputdelegate)
    (else
      (Teams|RefreshTeamChangeBinding _fortminigamelogic true _outputdelegate)
      (bind _asathena_player_controller_2 (Utilities|Casting|CastToAthena_PlayerController (Fortnite|GetFirstLocalFortPlayerController true)))
      (FortPlayerControllerAthena|UnbindEventfromOnLocalPlayerChangedTeams _asathena_player_controller_2 _outputdelegate))))

(fn ColorType
  ;[1 1 1] Knob_Color_Index  is either an array index into the Team Colors array, or if <0, is based on the Team of the target
  (return (select (>= (Variables|UserOption|GetKnob_ColorIndex) 0) "NewEnumerator1" (select (== (Variables|UserOption|GetKnob_ColorIndex) -1) "NewEnumerator2" "NewEnumerator3"))))

(fn CreateorUpdateMeshMID
  (Variables|Default|GetMesh%20MID
    (:then
      ;[1 1 1] If MID Exists but isn't attached, attach it. (Move Tool fix. This isn't a BuildingSMActor, so can't use Get Override Mesh Material
      (if (Utilities|NotEqual(Object) (Variables|Default|GetMesh%20MID) (Rendering|Material|GetMaterial (Variables|Components|GetStaticMeshComponent) 0))
        (Rendering|Material|SetMaterial (Variables|Components|GetStaticMeshComponent) 0 (Variables|Default|GetMesh%20MID))))
    (:else
      ;[1 1 1] If MID isn't already created, create it and save a ref
      (Variables|Default|SetMesh%20MID (Material|GetMID (Variables|Components|GetStaticMeshComponent))))))

(fn GetCreativeCustomEffectData
  (return 0 0 "None" 0 0))

(fn GetCreativeEffectData
  (return (Utilities|Struct|MakeFCreativeEffectDirectColor (CallFunction|GetTeamColor (Fortnite|GetFirstLocalFortPlayerController true)) (CallFunction|ColorType) (Variables|UserOption|GetKnob_IsVisibleToOwningPlayer) (Variables|UserOption|GetKnob_TeamRelationshipVisibility))))

(fn GetGameplayCue
  (if (Variables|Devoptions|GetUseDeveloperEffects)
    (bind _result (ArrayMacros|SafeArrayGet (Variables|DevOptions|GetDevGameplayEffect) (Variables|DevOptions|GetKnob_DevGameplayEffectIndex))
      (:%20%20
        (return _result)))
    (else
      (bind _result_2 (ArrayMacros|SafeArrayGet (Variables|Default|GetGameplayEffect) (Variables|UserOption|GetKnob_GameplayEffectIndex))
        (:%20%20
          (return _result_2))))))

(fn GetMeshComponent
  ;[1 1 1] Overriding Mesh Component with a new one. This doesn't use the base component as that doesn't interact with scale and the move tool properly
  ;
  (return 0))

(fn GetTeamColor
  (params (Actor%20to%20Check :type Actor%20Object%20Reference))
  (bind _result (ArrayMacros|SafeArrayGet (Class|GetClassDefaults "/Game/Creative/Sets/MilitaryBase/Props/CreativeTeamColors.CreativeTeamColors_C") (CallFunction|GetTeamColorIndex Actor%20to%20Check))
    (:%20%20
      (return _result))))

(fn GetTeamColorIndex
  (params (Actor%20to%20Check :type Actor%20Object%20Reference))
  (bind _actor_to_check (Variables|GetActor%20to%20Check))
  (switch Utilities|FlowControl|Switch|SwitchonECreativeColorSetType (CallFunction|ColorType)
    (:NewEnumerator0
      _)
    (:NewEnumerator1
      _)
    (:NewEnumerator2
      (bind _currentminigame (Variables|/Script/FortniteGame.FortMinigameLogicComponent|GetCurrentMinigame (Variables|Components|GetFortMinigameLogic))
        (:then
          (Utilities|IsValid _actor_to_check
            (:Is%20Valid
              (Utilities|IsValid (Minigame|GetMinigameTeamComponent _currentminigame)
                (:Is%20Valid
                  (return (Minigame|GameRules|GetTeamColorIndex (Minigame|GetMinigameTeamComponent _currentminigame) (Math|Byte|Byte+Byte (Utilities|Enum|LiteralenumEFortTeam "HumanPvP_Team1") (Team|GetActorPvPTeamIndex _actor_to_check)))))
                (:Is%20Not%20Valid
                  _)))
            (:Is%20Not%20Valid
              _))
          ;[1 1 1] Not using Team Color. Pass back the color index from the user option.
          (return (Variables|UserOption|GetKnob_ColorIndex)))
        (:else
          _)))
    (:NewEnumerator3
      (return (Math|Conversions|ToInteger(Byte) (Variables|Default|GetFriendly%20Color%20Index))))))

(fn UpdateColor
  (Networking|CanExecuteCosmeticEvents
    (:True
      (CallFunction|UpdatePowerupColors (CallFunction|GetTeamColorIndex (Fortnite|GetFirstLocalFortPlayerController true))))))

(fn UpdatePowerupColors
  (params (Index :type Integer))
  (bind (_teamlightcolors _teamselectorcolors) (Class|GetClassDefaults "/Game/Creative/Sets/MilitaryBase/Props/CreativeTeamColors.CreativeTeamColors_C"))
  (bind _mesh_mid (Variables|Default|GetMesh%20MID))
  (CallFunction|CreateOrUpdateMeshMID)
  (Components|Activation|Deactivate (Variables|Components|GetFX))
  (bind _result (ArrayMacros|SafeArrayGet _teamlightcolors (Variables|GetIndex))
    (:%20%20
      (Effects|Components|ParticleSystem|SetColorParameter (Variables|Components|GetFX) "Color" _result)
      (bind _result_2 (ArrayMacros|SafeArrayGet _teamselectorcolors (Variables|GetIndex))
        (:%20%20
          (LocalVariables|SetColor _result_2)
          (Rendering|Material|SetVectorParameterValue _mesh_mid "ColorHigh" (LocalVariables|GetColor) :!class /Script/Engine.MaterialInstanceDynamic)
          (Rendering|Material|SetVectorParameterValue _mesh_mid "Color_2" (Math|Color|LinearColor*Float (LocalVariables|GetColor) (Variables|Default|GetMesh%20Material%20Multiplier)) :!class /Script/Engine.MaterialInstanceDynamic)
          (Variables|Default|SetVFX%20Particle%20Color (LocalVariables|GetColor))
          (Components|Activation|Activate (Variables|Components|GetFX) true) ; Light Color of the FX doesn't update directly when you change it.\nThis forces the FX to reset.
        )))))

(fn ConstructionScript
  (|CallParentFunction)
  (CallFunction|UpdatePowerupColors 0))
