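# Unlit additive particle-sprite material that draws an animated, noise-distorted glowing ring, named for the invincibility powerup in Creative devices.
# Two panning noise samples warp the UVs, and nested sphere masks turn a radial gradient into the ring band.
# That mask blends ColorLow to particle color for emissive and feeds a distance-based opacity fade scaled by particle radius.
# A time-driven sine also builds a vertical bob and slight Z-axis rotation with corrected normals.
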Material /Game/Creative/Devices/Common/CreativePowerups/M_Invincible_Ring
graph {
 properties {
  bEnableResponsiveAA=true
  bUsedWithParticleSprites=true
  blendMode=BLEND_Additive
  pixelDepthOffsetMode=PDOM_Legacy
  shadingModel=MSM_Unlit
  translucencyPass=MTP_BeforeDOF
  usedShadingModels=Unlit
 }
 nodes [
  n1 = Sine(None=Multiply(A=Time(), B=Constant() {r=0.4}))
  n2 = Constant3Vector() {constant=<rgba 0 0 1 0>}
  n3 = LinearInterpolate(A=Constant() {r=-0.01}, B=Constant() {r=0.01}, Alpha=@n1)
  Add(
   A=LinearInterpolate(
    A=VectorParameter() {parameterName=BobHeightMax, defaultValue=<rgba 0 0 5 0>}.RGB
    B=VectorParameter() {parameterName=BobHeightMin, defaultValue=<rgba 0 0 -1 0>}.RGB
    Alpha=@n1
   )
   B=RotateAboutAxis(NormalizedRotationAxis=@n2, RotationAngle=@n3, PivotPoint=ObjectPositionWS(), Position=WorldPosition().XYZ)
  )
  n4 = AppendVector(
   A=TextureSample(UVs=Panner() {speedX=0.15, speedY=-0.038}) {texture=&T_NoisePack_01}.R
   B=TextureSample(UVs=Panner() {speedX=-0.077, speedY=0.371}) {texture=&T_NoisePack_01}.G
  )
  n5 = Add(
   A=Multiply(A=ScalarParameter() {defaultValue=0.015, parameterName=NoiseScale}, B=ConstantBiasScale(None=@n4) {bias=-0.5, scale=2})
   B=TextureCoordinate()
  )
  n6 = Constant() {r=0.49}
  n7 = Constant() {r=0.5}
  n8 = Constant()
  n9 = SphereMask(
   A=Power(Base=SphereMask(A=@n5, B=@n6, Radius=@n7, Hardness=@n8), Exp=ScalarParameter() {defaultValue=1.75, parameterName=Exp})
   B=ScalarParameter() {defaultValue=0.125, parameterName=Center}
   Radius=ScalarParameter() {defaultValue=0.08, parameterName=Radius}
   Hardness=ScalarParameter() {parameterName=Hardness}
  )
  n10 = ScalarParameter() {defaultValue=4, parameterName=ColorLerpExponent}
  MaterialFunctionCall(Emissive=LinearInterpolate(
    A=VectorParameter() {parameterName=ColorLow, defaultValue=<rgba 0.0302 0.0581 0.3 0>}.RGB
    B=ParticleColor().RGB
    Alpha=Power(Base=@n9, Exp=@n10)
   )) {materialFunction=&MF_FX_EmissiveAdjustments}
  n11 = Add(A=ParticleRadius(), B=ScalarParameter() {defaultValue=80, parameterName=RadiusBias})
  MaterialFunctionCall(
   Opacity In=@n9
   Position=ParticlePositionWS()
   Fade_Start_Distance=Add(A=ScalarParameter() {defaultValue=20, parameterName="Fade_Start_Distance (S)"}, B=@n11)
   Fade_End_Distance=@n11
  ) {materialFunction=&MF_Distance_SyncedFade_FX}
  MaterialFunctionCall(Rotation Axis=@n2, Rotation Angle=@n3) {materialFunction=&FixRotateAboutAxisNormals}
  VectorParameter() {parameterName=ColorHigh, defaultValue=<rgba 3 0.4069 2.3939 0>}
 ]
}
