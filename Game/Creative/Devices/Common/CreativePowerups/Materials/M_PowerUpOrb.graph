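# Masked material for the Creative powerup orb: a glassy, faceted sphere with animated circuitry, tinted by ColorHigh.
# Normal-based facets, inverse Fresnel and a pulsing height mask shape the base color, while circuit masks drive metallic, roughness, Fresnel glow and traveling emissive surges.
# A sine-driven world position offset bobs the orb, and a distance fade drives dithered opacity.
# The attributes feed MF_CreativeMoveTool; the rotation branch is unused.
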Material /Game/Creative/Devices/Common/CreativePowerups/Materials/M_PowerUpOrb
graph {
 properties {bUseMaterialAttributes=true, blendMode=BLEND_Masked, ditherOpacityMask=true, pixelDepthOffsetMode=PDOM_Legacy, usedShadingModels=Default Lit}
 nodes [
  n1 = Sine(None=Multiply(A=Time(), B=Constant() {r=0.4}))
  n2 = LinearInterpolate(
   A=VectorParameter() {parameterName=BobHeightMax, defaultValue=<rgba 0 0 5 0>}.RGB
   B=VectorParameter() {parameterName=BobHeightMin, defaultValue=<rgba 0 0 -1 0>}.RGB
   Alpha=@n1
  )
  n3 = Constant() {r=-0.01}
  n4 = Constant() {r=0.01}
  Add(
   A=@n2
   B=RotateAboutAxis(
    NormalizedRotationAxis=Constant3Vector() {constant=<rgba 0 0 1 0>}
    RotationAngle=LinearInterpolate(A=@n3, B=@n4, Alpha=@n1)
    PivotPoint=ObjectPositionWS()
    Position=WorldPosition().XYZ
   )
  )
  n5 = VectorParameter() {parameterName=ColorHigh, defaultValue=<rgba 1 1 0 0>}
  n6 = Multiply(A=TextureCoordinate() {uTiling=2}, B=ScalarParameter() {defaultValue=3, parameterName=Tiling})
  n7 = LinearInterpolate(
   A=@n5.RGB
   B=TextureSample(UVs=@n6) {texture=/Game/Creative/Devices/Common/CreativePowerups/Textures/T_PowerUp_BC}.RGB
   Alpha=ScalarParameter() {defaultValue=0.9, parameterName=Color Variation}
  )
  n8 = Multiply(A=@n5.RGB, B=ScalarParameter() {defaultValue=0.2, parameterName=B})
  n9 = TextureObjectParameter() {parameterName=Normal, samplerType=SAMPLERTYPE_Normal, texture=/Game/Creative/Devices/Common/CreativePowerups/Textures/T_PowerUp_N}
  n10 = Transform(None=TextureSample(UVs=@n6, Tex=@n9) {samplerType=SAMPLERTYPE_Normal, texture=&T_PowerUp_N}.RGB)
  n11 = CameraVectorWS()
  n12 = Multiply(A=Saturate(None=DotProduct(A=@n10, B=@n11)), B=ScalarParameter() {defaultValue=50, parameterName=Facet})
  n13 = Fresnel(ExponentIn=ScalarParameter() {defaultValue=0.17, parameterName=ExponentIn}) {baseReflectFraction=0}
  n14 = OneMinus(None=@n13)
  n15 = Multiply(A=Time()) {constB=0.2}
  n16 = TextureObjectParameter() {
   parameterName=HeightMap
   samplerType=SAMPLERTYPE_LinearColor
   texture=/Game/Creative/Devices/Common/CreativePowerups/Textures/T_PowerUp_Mask02_M
  }
  n17 = TextureSample(UVs=@n6, Tex=@n16) {samplerType=SAMPLERTYPE_LinearColor, texture=&T_PowerUp_Mask02_M}
  n18 = ConstantBiasScale(None=Sine(None=Add(A=@n15, B=@n17.G)))
  n19 = ScalarParameter() {defaultValue=2.5, parameterName=Height Exp}
  n20 = Saturate(None=MaterialFunctionCall(In=@n13, Contrast=ScalarParameter() {defaultValue=0.05, parameterName="Contrast (S)"}) {materialFunction=&CheapContrast}.Result)
  n21 = TextureCoordinate()
  n22 = Sine(None=Multiply(A=ComponentMask(None=@n21) {g=true}) {constB=0.5})
  n23 = Multiply(A=Multiply(A=Multiply(A=@n18, B=@n17.B), B=Power(Base=@n17.R, Exp=@n19)), B=Multiply(A=OneMinus(None=@n20), B=Multiply(A=@n22, B=@n22)))
  n24 = LinearInterpolate(A=Multiply(A=@n5.RGB) {constB=0.02}, B=Multiply(A=Add(A=@n7, B=@n8), B=Multiply(A=@n12, B=@n14)), Alpha=@n23)
  n25 = TextureSample(UVs=TextureCoordinate() {uTiling=2}) {samplerType=SAMPLERTYPE_LinearColor, texture=/Game/Creative/Devices/Common/CreativePowerups/Textures/T_PowerUp_Mask01_M}
  n26 = Multiply(A=Saturate(None=Subtract(A=@n24) {constB=0.1}), B=ScalarParameter() {parameterName=Emissive})
  n27 = ScalarParameter() {defaultValue=2, parameterName=Glow Fresnel Exponent}
  n28 = Multiply(A=Time()) {constB=0.333}
  n29 = ConstantBiasScale(None=Cosine(None=Add(A=@n28, B=@n25.G)))
  n30 = Multiply(
   A=Multiply(A=@n5.RGB, B=Fresnel(ExponentIn=@n27) {baseReflectFraction=0, exponent=3})
   B=Add(A=Multiply(A=@n25.R, B=@n29), B=ScalarParameter() {defaultValue=1, parameterName=Fresnel Glow})
  )
  n31 = Time()
  n32 = ScalarParameter() {defaultValue=0.3, parameterName=Surge Speed}
  n33 = Frac(None=Add(A=Multiply(A=@n31, B=@n32) {constB=0.1}, B=@n25.A))
  n34 = Subtract(A=Multiply(A=Saturate(None=@n33)) {constB=3}) {constB=2}
  n35 = Sine(None=Saturate(None=Add(A=@n34, B=@n25.B)))
  n36 = Multiply(A=Multiply(A=Saturate(None=@n35), B=Multiply(A=@n25.R, B=@n25.A)), B=@n5.RGB)
  n37 = ScalarParameter() {defaultValue=1, parameterName=Surge Intensity}
  n38 = ScalarParameter() {defaultValue=20, parameterName="Fade_Start_Distance (S)"}
  n39 = Add(A=ObjectRadius(), B=ScalarParameter() {defaultValue=100, parameterName=RadiusBias})
  n40 = SetMaterialAttributes!(
   MaterialAttributes
   Base Color=@n24
   Metallic=Add(A=@n25.R, B=Add(A=@n14, B=@n23))
   Roughness=LinearInterpolate(
    A=ScalarParameter() {defaultValue=0.05, parameterName=Roughness Glass}
    B=ScalarParameter() {defaultValue=0.5, parameterName=Roughness Circuit}
    Alpha=@n25.R
   )
   Emissive Color=Add(A=Add(A=@n26, B=@n30), B=Multiply(A=@n36, B=@n37))
   Opacity Mask=MaterialFunctionCall(Position=ObjectPositionWS(), Fade_Start_Distance=Add(A=@n38, B=@n39), Fade_End_Distance=@n39) {materialFunction=&MF_Distance_SyncedFade_FX}.Result
   World Position Offset=@n2
  ) {attributeSetTypes [
    #1
    #2
    #3
    #4
    #5
    #6
   ]}
  MaterialFunctionCall(MA In=@n40) {materialFunction=&MF_CreativeMoveTool}
  MaterialFunctionCall!(Mask=Constant() {r=1}, Index) {materialFunction=/Game/Creative/Materials/MF_CreativeColors}
  ParticleColor()
 ]
}
