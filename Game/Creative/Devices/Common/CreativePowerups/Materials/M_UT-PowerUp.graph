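# Lit material for Creative power-up devices that spins and bobs the mesh with a glowing, animated surface.
# Emissive combines inverted Fresnel, noise-distorted panning caustics and swept scanlines, with custom primitive data offsetting their timing.
# A bounding-box height gradient and local-Y facing mask shape Base Color and emissive.
# World Position Offset adds Z-axis rotation to MF_Creative_CoinBobble, normals follow the spin, and MF_CreativeMoveTool wraps the final attributes.
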Material /Game/Creative/Devices/Common/CreativePowerups/Materials/M_UT-PowerUp
graph {
 properties {bUseMaterialAttributes=true, pixelDepthOffsetMode=PDOM_Legacy, usedShadingModels=Default Lit}
 nodes [
  n1 = VectorParameter() {parameterName=Base Color, defaultValue=<rgba 0.8508 0.0594 0.9253 0>}
  n2 = MaterialFunctionCall() {materialFunction=&BoundingBoxBased_0-1_UVW}
  n3 = VertexNormalWS()
  n4 = Saturate(None=Transform(None=Constant3Vector() {constant=<rgba 0 1 0 0>}) {transformSourceType=TRANSFORMSOURCE_Local})
  n5 = Power(Base=Abs(None=DotProduct(A=@n3, B=@n4))) {constExponent=10}
  n6 = ScalarParameter() {defaultValue=2, parameterName=FresnelExponent}
  n7 = Normalize(VectorInput=Constant3Vector() {constant=<rgba 0 0 1 0>})
  n8 = Divide(A=Time() {bOverride_Period=true, period=5}) {constB=5}
  n9 = Add(A=RotateAboutAxis(NormalizedRotationAxis=@n7, RotationAngle=@n8, PivotPoint=Constant(), Position=@n3), B=@n3)
  n10 = VectorParameter() {parameterName=DistortionATilePan, defaultValue=<rgba 0.5 1 0.078 0.123>}
  n11 = Multiply(A=TextureCoordinate(), B=AppendVector(A=@n10.R, B=@n10.R))
  n12 = ScalarParameter() {bUseCustomPrimitiveData=true, parameterName=TimeOffset, primitiveDataIndex=31}
  n13 = Add(A=Time(), B=@n12)
  n14 = Constant()
  n15 = Panner(Coordinate=@n11, Time=@n13, Speed=AppendVector(A=@n10.B, B=@n14))
  n16 = Panner(Coordinate=@n11, Time=@n13, Speed=AppendVector(A=@n14, B=@n10.A))
  n17 = Multiply(
   A=AppendVector(
    A=TextureSample(UVs=@n15) {samplerType=SAMPLERTYPE_LinearColor, texture=&T_NoisePack_LinearColor}.R
    B=TextureSample(UVs=@n16) {samplerType=SAMPLERTYPE_LinearColor, texture=&T_NoisePack_LinearColor}.G
   )
   B=ScalarParameter() {defaultValue=0.2, parameterName=NoiseADistortionScale}
  )
  n18 = TextureCoordinate()
  n19 = VectorParameter() {parameterName=NoiseATilePan, defaultValue=<rgba 1 1 0.078 0.123>}
  n20 = Panner(Coordinate=Multiply(A=Multiply(A=@n17, B=@n18), B=AppendVector(A=@n19.R, B=@n19.R)), Time=@n13, Speed=AppendVector(A=@n19.B, B=@n19.B))
  n21 = VectorParameter() {parameterName=DistortionBTilePan, defaultValue=<rgba 0.89 0.89 -0.1488 -0.11>}
  n22 = Multiply(A=TextureCoordinate(), B=AppendVector(A=@n21.R, B=@n21.R))
  n23 = Constant()
  n24 = Panner(Coordinate=@n22, Time=@n13, Speed=AppendVector(A=@n21.B, B=@n23))
  n25 = Panner(Coordinate=@n22, Time=@n13, Speed=AppendVector(A=@n23, B=@n21.A))
  n26 = Multiply(
   A=AppendVector(
    A=TextureSample(UVs=@n24) {samplerType=SAMPLERTYPE_LinearColor, texture=&T_NoisePack_LinearColor}.G
    B=TextureSample(UVs=@n25) {samplerType=SAMPLERTYPE_LinearColor, texture=&T_NoisePack_LinearColor}.B
   )
   B=ScalarParameter() {defaultValue=0.25, parameterName=NoiseBDistortionScale}
  )
  n27 = TextureCoordinate()
  n28 = VectorParameter() {parameterName=NoiseBTilePan, defaultValue=<rgba 0.8 0.88 -0.133 -0.097>}
  n29 = Panner(Coordinate=Multiply(A=Multiply(A=@n26, B=@n27), B=AppendVector(A=@n28.R, B=@n28.R)), Time=@n13, Speed=AppendVector(A=@n28.B, B=@n28.B))
  n30 = Power(
   Base=Add(
    A=TextureSample(UVs=@n20) {texture=&T_ShieldCaustics}.R
    B=TextureSample(UVs=@n29) {texture=&T_ShieldCaustics}.R
   )
   Exp=ScalarParameter() {defaultValue=1.5, parameterName=Exp}
  )
  n31 = ComponentMask(None=TextureCoordinate()) {g=true}
  n32 = Frac(None=Multiply(A=@n31, B=ScalarParameter() {defaultValue=15, parameterName=Steps}))
  n33 = Constant() {r=0.5}
  n34 = Multiply(
   A=ScalarParameter() {defaultValue=0.2, parameterName=ScanlinesRadius}
   B=MaterialFunctionCall() {materialFunction=&BoundingBoxBased_0-1_UVW}.B
  )
  n35 = ScalarParameter() {defaultValue=0.6, parameterName=ScanlinesHardness}
  n36 = Subtract(A=ComponentMask(None=@n31) {g=true}) {constB=0.5}
  n37 = Time() {period=1}
  n38 = Frac(None=Multiply(A=Add(A=@n12, B=@n37), B=ScalarParameter() {defaultValue=0.2, parameterName=Speed}))
  n39 = Multiply(A=ConstantBiasScale(None=OneMinus(None=@n38)) {bias=-0.5, scale=2}) {constB=4}
  n40 = ScalarParameter() {defaultValue=1, parameterName=SweepRadius}
  n41 = ScalarParameter() {parameterName=SweepHardness}
  n42 = Multiply(
   A=Multiply(A=SphereMask(A=@n32, B=@n33, Radius=@n34, Hardness=@n35), B=SphereMask(A=@n36, B=@n39, Radius=@n40, Hardness=@n41))
   B=ScalarParameter() {defaultValue=0.55, parameterName=ScanlinesOpacity}
  )
  n43 = Multiply(
   A=Add(
    A=MaterialFunctionCall(Exponent=@n6, Vertex / Pixel Normal=@n9) {materialFunction=&MF_Fresnel_FX}.Fresnel Inverted
    B=Add(A=@n30, B=@n42)
   )
   B=@n2.B
  )
  n44 = VectorParameter() {parameterName=EmissiveColor, defaultValue=<rgba 2 0.2299 1.8783 0>}
  n45 = MaterialFunctionCall() {materialFunction=&ObjectPivotPoint}
  n46 = WorldPosition()
  n47 = SetMaterialAttributes!(
   MaterialAttributes
   Base Color=Multiply(A=Multiply(A=@n1.RGB, B=@n2.B), B=@n5)
   Metallic=Multiply(A=ScalarParameter() {defaultValue=0.85, parameterName=Metallic}, B=@n5)
   Specular=ScalarParameter() {defaultValue=1, parameterName=Specular}
   Roughness=ScalarParameter() {defaultValue=0.3, parameterName=Roughness}
   Emissive Color=Multiply(A=Multiply(A=@n43, B=@n44.RGB), B=@n5)
   World Position Offset=Add(
    A=MaterialFunctionCall!(TimeScale, MaxHeight, MinHeight) {materialFunction=/Game/Playgrounds/Items/Props/Materials/MF_Creative_CoinBobble}.WorldPositionOffset
    B=RotateAboutAxis(NormalizedRotationAxis=@n7, RotationAngle=@n8, PivotPoint=@n45.Object Pivot Location, Position=@n46.XYZ)
   )
   Normal=@n9
  ) {attributeSetTypes [
    #1
    #2
    #3
    #4
    #5
    #6
    #7
   ]}
  MaterialFunctionCall(MA In=@n47) {materialFunction=&MF_CreativeMoveTool}
 ]
}
